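# Niagara system for the QuickBurst backpack cosmetic, adding flickering eye glows and exhaust flames, smoke, sparks and lights at its pipes.
# Emitters partition particles by PipeAmount and sample skeletal mesh sockets for placement.
# System timelines and burst values drive spawn rates, color and intensity, including separate end smoke.
# Smoke ribbons blend toward the moving socket, inherit system velocity and tint by distance, while variant selectors provide alternate colors.
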NiagaraSystem /BRCosmetics/Accessories/FORT_Backpacks/Backpack_M_MED_QuickBurst/FX/NS_Backpack_QuickBurst
dependencies {
 usedDataInterfaces [
  &NiagaraDataInterfaceArrayInt32
  &NiagaraDataInterfaceArrayFloat3
  &NiagaraDataInterfaceCurve
  &NiagaraDataInterfaceVector2DCurve
  &NiagaraDataInterfaceColorCurve
  &NiagaraDataInterfacePlatformSet
  &NiagaraDataInterfaceVectorCurve
  &NiagaraDataInterfaceArrayFloat
  &NiagaraDataInterfaceSkeletalMesh
 ]
 usedDynamicInputs [
  &MakeFloatFromVector2d
  &Boolean_OrOperation
  &Waveform
 ]
 usedModules [
  &FN_VariantManager
  &self:SystemSpawnScript.SystemScriptSource.SystemScriptGraph.NiagaraNodeAssignment_1.SetVariables_#1
  &SystemState
  &Timeline
  &self:SystemSpawnScript.SystemScriptSource.SystemScriptGraph.NiagaraNodeAssignment_0.SetVariables_#2
  &self:SystemSpawnScript.SystemScriptSource.SystemScriptGraph.NiagaraNodeAssignment_2.SetVariables_#3
  &self:SystemSpawnScript.SystemScriptSource.SystemScriptGraph.NiagaraNodeAssignment_3.SetVariables_#4
 ]
}
emitters {
 End_Smoke {
  emitterUpdateScript [
   &EmitterState()
   &SpawnRate(
    SpawnRate=&self:OneWayTimelineScaler(OneWayTimeline:NiagaraFloat=$System.OneWayEndBurst, Scale:NiagaraFloat=10)
    Use Spawn Probability=true
   )
  ]
  particleSpawnScript [
   &PartitionParticles(Number Of Partitions=$System.PipeAmount, Write Partition to Ribbon ID=::NewEnumerator0)
   &InitializeParticle(
    Color=&MultiplyLinearColor(
     Linear Color=<rgba 1 1 1 1.5>
     Scale Factor=&MakeVector4_FromFloat(Float=&FN_FrontEndFloatModifier(Gameplay=1, Front End=1))
    )
    Lifetime=&RandomRangeFloat(Minimum=0.5, Maximum=0.8)
    Mesh Scale=&Multiply_VectorByFloat(Vector=&RandomRangeVector(Minimum=<3.5 2 2.5>, Maximum=<4.5 3.5 3.5>), Float=0.02)
    Mesh Scale Mode=::NewEnumerator4
    UsePositionOffset=false
    Write Material Random=true
   )
   &SkeletalMeshLocation(
    Bone / Socket Index=&SelectIntFromArray(Direct Array Index=$Particles.PartitionParticles.Partition, Int Selection Array=DataInterface() {propertyValues {IntData [2,3,4]}})
    MaskSpawnGroup=false
    Mesh Sampling Type=::NewEnumerator5
    Orientation Sampling=::NewEnumerator1
    Skeletal Mesh=$System.SkeletalMesh
    Socket Sampling Mode=::NewEnumerator5
   ) {module=SkeletalMeshLocation001}
   &InitialMeshOrientation(
    Mesh Orientation Mode=::NewEnumerator2
    Rotation=&RandomRangeVector(Minimum=<0 0 0>, Maximum=<1 0 0>)
    Use Rotation Vector=true
   )
   &CameraOffset(Camera Offset Amount=-10)
  ]
  particleUpdateScript [
   &ParticleState()
   &SkeletalMeshLocation(
    Bone / Socket Index=&SelectIntFromArray(Direct Array Index=$Particles.PartitionParticles.Partition, Int Selection Array=DataInterface() {propertyValues {IntData [2,3,4]}})
    MaskSpawnGroup=false
    Mesh Sampling Type=::NewEnumerator5
    Orientation Sampling=::NewEnumerator1
    Position Sampling=::NewEnumerator2
    Sampled Position Offset=<0 0 -1>
    Skeletal Mesh=$System.SkeletalMesh
    Socket Sampling Mode=::NewEnumerator5
   ) {module=SkeletalMeshLocation002}
   &InheritSourceMovement(
    Applied Force Scale=&VectorFromFloat(Value=&FloatFromCurve(FloatCurve=DataInterface() {propertyValues {Curve {keys [<at 0.7 1>,<at 0.8 0>]}}}, Scale Curve=1.2))
    Applied Position Scale=&Multiply_VectorByFloat(
     Vector=<1 1 1>
     Float=&FloatFromCurve(FloatCurve=DataInterface() {propertyValues {Curve {keys [
          <at 0 1 RCIM_Cubic>
          {arriveTangent=-0.625, interpMode=RCIM_Cubic, leaveTangent=-0.625, time=0.2, value=1}
          {arriveTangent=-0.0269, interpMode=RCIM_Cubic, leaveTangent=-0.0269, tangentMode=RCTM_User, time=0.8, value=0.5}
         ]}}})
    )
    ApplyForce=true
    ApplyForceFalloff=false
    ApplyPosition=true
    ApplyPositionFalloff=false
    Module.LimitForceSpeed=false
    Source Velocity:Vector3f=$System.SystemVelocity
   )
   &Drag(Drag=8, Use Linear Drag=true, Use Rotational Drag=true)
   &GravityForce(Gravity=<0 0 200>)
   &SolveForcesAndVelocity(Clamp Velocity=false, Limit Acceleration=false)
   &ScaleMeshSize(Scale Factor=&Multiply_VectorByFloat(
     Vector=&Multiply_VectorByFloat(
      Vector=&Multiply_Vector(
       A=<1 1 1>
       B=&VectorFromCurve(VectorCurve=DataInterface() {propertyValues {XCurve {keys [<at 0 5>,<at 0.2 1>,<at 1 1>]}, YCurve {keys [<at 0 1>,<at 1 1>]}, ZCurve {keys [<at 0 1>,<at 1 1>]}}})
      )
      Float=&RandomRangeFloat(Minimum=1, Maximum=1.5)
     )
     Float=&FloatFromCurve(
      FloatCurve=DataInterface() {propertyValues {Curve {keys [
          {interpMode=RCIM_Cubic, tangentMode=RCTM_User, time=0, value=0}
          {arriveTangent=1.3433, tangentMode=RCTM_User, time=0.2, value=0.35}
          {arriveTangent=1.4526, interpMode=RCIM_Cubic, leaveTangent=1.4526, tangentMode=RCTM_User, time=1, value=1.5}
         ]}}}
      Scale Curve=&Multiply_Float(A=1, B=1.5)
     )
    ))
   &DynamicMaterialParameters(
    Index 0 Param 1=&FloatFromCurve(
     FloatCurve=DataInterface() {propertyValues {Curve {keys [
         {interpMode=RCIM_Cubic, tangentMode=RCTM_User, time=0, value=0}
         <at 0.2 0.55>
         {arriveTangent=-1.0223, interpMode=RCIM_Cubic, leaveTangent=-1.0223, tangentMode=RCTM_User, time=1, value=1}
        ]}}}
     Scale Curve=&RandomRangeFloat(Minimum=7, Maximum=15)
    )
    Index 0 Param 4=&Lerp_Float(
     A=0.6
     B=0
     Alpha=&FloatFromCurve(FloatCurve=DataInterface() {propertyValues {Curve {keys [
          {arriveTangent=0.7438, interpMode=RCIM_Cubic, leaveTangent=0.7438, tangentMode=RCTM_User, time=0, value=0}
          {arriveTangent=0.0023, interpMode=RCIM_Cubic, leaveTangent=0.0023, tangentMode=RCTM_User, time=1, value=1}
         ]}}})
    )
    Module.Index0Param0Write=true
    Module.Index0Param1Write=true
    Module.Index0Param2Write=true
    Module.Index0Param3Write=true
   ) {module=DynamicMaterialParameters001}
   /SetVariables(
    Particles.DY0:Vector4f=<4 0 0 0>
    Particles.MaterialRandom:NiagaraFloat=&SelectFloatFromArray(Direct Array Index=$Particles.PartitionParticles.Partition, Float Selection Array=DataInterface() {propertyValues {FloatData [0,0.2,0.4]}})
    Particles.Position=&Lerp_Vector(
     A=&AddVectorToPosition(
      Position=$Particles.SkeletalMeshLocation002.SampledPosition
      Vector=&RotateVectorByQuaternion(
       Vector To Rotate=&VectorFromCurve(
        VectorCurve=DataInterface() {propertyValues {XCurve {keys [<at 0 0>,<at 1 0>]}, YCurve {keys [<at 0 0>,<at 1 0>]}, ZCurve {keys [<at 0 0>,<at 1 1>]}}}
        Scale Curve=<1 1 10>
       )
       Quaternion=$Particles.SkeletalMeshLocation002.SampledRotation
      )
     )
     B=Linked() {linkedVariable {name=Particles.Position, type=NiagaraPosition}}
     Alpha=&FloatFromCurve(FloatCurve=DataInterface() {propertyValues {Curve {keys [{interpMode=RCIM_Cubic, tangentMode=RCTM_User, time=0, value=0.2}, {interpMode=RCIM_Cubic, tangentMode=RCTM_User, time=0.7, value=1}]}}})
    )
    Particles.RibbonColor:LinearColor=&MultiplyLinearColor(
     Linear Color=&ColorFromCurve(ColorCurve=DataInterface() {propertyValues {
        AlphaCurve {keys [<at 0.1579 2>,<at 0.656 1>,<at 0.8476 0>]}
        BlueCurve {keys [<at 0 2.5 RCIM_Cubic>,<at 0.1632 0>,<at 0.5591 0 RCIM_Cubic>]}
        GreenCurve {keys [<at 0 8.75 RCIM_Cubic>,<at 0.1632 4.2629>,<at 0.5591 0.1667 RCIM_Cubic>]}
        RedCurve {keys [<at 0 40 RCIM_Cubic>,<at 0.1632 40>,<at 0.5591 0.5 RCIM_Cubic>]}
       }})
     Scale Factor=&MakeVector4(X=1, Y=1, Z=1, W=$System.Timeline.NormalizedTime)
    )
    Particles.RibbonSmokeColor:LinearColor=&MultiplyLinearColor(
     Linear Color=&ColorFromCurve(ColorCurve=DataInterface() {propertyValues {
        AlphaCurve {keys [<at 0.1579 4>,<at 0.656 1>,<at 0.8476 0>]}
        BlueCurve {keys [<at 0 2.5 RCIM_Cubic>,<at 0.1584 0 RCIM_Cubic>]}
        GreenCurve {keys [<at 0 8.75 RCIM_Cubic>,<at 0.1584 0.1667 RCIM_Cubic>]}
        RedCurve {keys [<at 0 40 RCIM_Cubic>,<at 0.1584 0.5 RCIM_Cubic>]}
       }})
     Scale Factor=&MakeVector4(X=1, Y=1, Z=1, W=$System.Timeline.NormalizedTime)
    )
    Particles.RibbonSmokeW:NiagaraFloat=&FloatFromCurve(FloatCurve=DataInterface() {propertyValues {Curve {keys [<at 0 2.5>,<at 1 6>]}}}, Scale Curve=2)
    Particles.RibbonWidth=&FloatFromCurve(FloatCurve=DataInterface() {propertyValues {Curve {keys [<at 0 2.5>,<at 1 6>]}}}, Scale Curve=2)
   ) {isSetParameters=true, module=SetVariables_#5}
  ]
  rendererClasses [&NiagaraMeshRendererProperties]
  renderers [{rendererClass=&NiagaraMeshRendererProperties, rendererIndex=0}]
 }
 EyeGlowStack {
  emitterUpdateScript [
   &EmitterState(
    Inactive Response=::NewEnumerator1
    Life Cycle Mode=::NewEnumerator1
    Loop Behavior=::NewEnumerator1
    Loop Duration Mode=::NewEnumerator1
    UseLoopDelay=true
   )
   /Niagara/Noise/Waveform(
    EnableExponent1=false
    GlobalClampMax=false
    GlobalClampMin=false
    UseInput1=false
    UsePhase1=false
    "Waveform [1]":ENiagara_Waveforms=::NewEnumerator12
    "[1] Frequency"=7
   )
   /SetVariables(
    Emitter.Flicker:NiagaraFloat=&Add_Float(A=&Multiply_Float(A=$Emitter.Waveform.Waveform, B=0.5), B=0.7)
    Emitter.FlickerHeavy:NiagaraFloat=&Add_Float(A=&Multiply_Float(A=$Emitter.Waveform.Waveform, B=2), B=0.8)
   ) {isSetParameters=true}
   &SpawnBurst_Instantaneous(Spawn Count=2, Use Spawn Probability=false, UseLoopCountLimit=true)
   /SetVariables(
    Emitter.CoreGlowColor:LinearColor=&MultiplyLinearColor(
     Linear Color=&VariantColorSelector(Variant Amount=0, Original Color=<rgba 1 0.5439 0 1>, Variant 1 Color=<rgba 0 0.3873 1 1>, Variant 2 Color=?)
     Scale Factor=&MakeVector4_FromFloat(Float=&Multiply_Float(A=1, B=$Emitter.Flicker))
    )
    Emitter.CoreGlowSize:Vector2f=&Vector2DFromFloat(Value=1.407)
    Emitter.DarkGlowColor:LinearColor=&MultiplyLinearColor(
     Linear Color=&VariantColorSelector(Variant Amount=0, Original Color=<rgba 0.1812 0.0405 0.017 1.5>, Variant 1 Color=<rgba 0 0.074 0.1812 1.5>, Variant 2 Color=?)
     Scale Factor=&MakeVector4_FromFloat(Float=&Multiply_Float(A=1, B=$Emitter.Flicker))
    )
    Emitter.DarkGlowSize:Vector2f=&Vector2DFromFloat(Value=8.0622)
    Emitter.DepthFade=1
    Emitter.DiffuseGlowColor:LinearColor=&MultiplyLinearColor(
     Linear Color=&VariantColorSelector(Variant Amount=0, Original Color=<rgba 2 0.5832 0 1>, Variant 1 Color=<rgba 0 2.4025 4 1>, Variant 2 Color=?)
     Scale Factor=&MakeVector4_FromFloat(Float=&Multiply_Float(A=1, B=$Emitter.Flicker))
    )
    Emitter.DiffuseGlowSize:Vector2f=&Vector2DFromFloat(Value=3.4401)
    Emitter.LightExponent=5
    Emitter.LightGlowColor:LinearColor=&MultiplyLinearColor(
     Linear Color=&VariantColorSelector(Variant Amount=0, Original Color=<rgba 1.5 0.2154 0 1>, Variant 1 Color=<rgba 0.0352 0.1636 1 1>, Variant 2 Color=?)
     Scale Factor=&MakeVector4_FromFloat(Float=&Multiply_Float(A=7, B=$Emitter.FlickerHeavy))
    )
    Emitter.LightRadius=3
   ) {isSetParameters=true}
  ]
  particleSpawnScript [
   &PartitionParticles()
   &InitializeParticle(Color=<rgba 1 0.3963 0 0.7>, Sprite Size Mode=::NewEnumerator3, Uniform Sprite Size=120, UsePositionOffset=false, Write Material Random=true)
   &DynamicMaterialParameters(
    Index 0 Param 1=0.2
    Index 0 Param 2=1
    Module.Param0WriteEnabled=true
    Module.Param1WriteEnabled=true
    Module.Param2WriteEnabled=true
    Module.Param3WriteEnabled=true
   )
  ]
  particleUpdateScript [
   &ParticleState(Kill Particles When Lifetime Has Elapsed=false, Let Infinitely Lived Particles Die When Emitter Deactivates=true)
   &SkeletalMeshLocation(
    Bone / Socket Index=$Particles.PartitionParticles.Partition
    MaskSpawnGroup=false
    Mesh Sampling Type=::NewEnumerator5
    Skeletal Mesh=$System.SkeletalMesh
    Socket Sampling Mode=::NewEnumerator5
   )
   &ScaleColor(Scale Alpha=&Multiply_Float(A=$Emitter.Flicker, B=0.3), ScaleA=true, ScaleRGB=true)
   &FN_Light_Attributes(
    Color=$Particles.Color
    Create Color Attribute=true
    Create Exponent Attribute=true
    Create Radius Attribute=true
    Exponent=5
    Radius=4
    Scale Color=&Multiply_Float(A=7, B=$Emitter.Flicker)
   )
  ]
  rendererClasses [&NiagaraSpriteRendererProperties, &NiagaraLightRendererProperties]
  renderers [
   {rendererClass=&NiagaraSpriteRendererProperties}
   {rendererClass=&NiagaraSpriteRendererProperties, rendererIndex=1}
   {rendererClass=&NiagaraSpriteRendererProperties, rendererIndex=2}
   {rendererClass=&NiagaraLightRendererProperties, rendererIndex=3}
  ]
 }
 Flames {
  emitterUpdateScript [
   &EmitterState()
   &SpawnRate(Spawn Probability=$System.Timeline.Time, SpawnRate=200, Use Spawn Probability=true) {module=SpawnRate002}
  ]
  particleSpawnScript [
   &PartitionParticles(Number Of Partitions=$System.PipeAmount)
   &FN_SetSpawnGroup(Number of Spawn Groups=2)
   &InitializeParticle(
    Color=&VariantColorSelector(Variant Amount=0, Original Color=<rgba 1 1 1 1>, Variant 1 Color=<rgba 40 40 40 1>, Variant 2 Color=?)
    Lifetime=&RandomRangeFloat(Minimum=0.35, Maximum=0.55)
    Sprite Size Mode=::NewEnumerator3
    Sprite UV Mode=::NewEnumerator1
    Uniform Sprite Size=1.7497
    UsePositionOffset=false
    Write Material Random=true
   )
   &SkeletalMeshLocation(
    Bone / Socket Index=&SelectIntFromArray(Direct Array Index=$Particles.PartitionParticles.Partition, Int Selection Array=DataInterface() {propertyValues {IntData [2,3,4]}})
    MaskSpawnGroup=false
    Mesh Sampling Type=::NewEnumerator5
    Orientation Sampling=::NewEnumerator1
    Skeletal Mesh=$System.SkeletalMesh
    Socket Sampling Mode=::NewEnumerator5
   )
   &AddVelocity(
    Cone Angle=0
    Cone Axis=&RotateVectorByQuaternion(Vector To Rotate=<0 0 1>, Quaternion=$Particles.SkeletalMeshLocation.SampledRotation)
    Module.UseConeFalloff=false
    Module.UseVelDistribution=false
    Override Seed=false
    Velocity Mode=::NewEnumerator2
    Velocity Speed=&Multiply_Float(A=&RandomRangeFloat(Minimum=0.5, Maximum=1), B=20)
   ) {module=AddVelocity001}
   &CameraOffset()
  ]
  particleUpdateScript [
   &ParticleState()
   &ScaleSpriteSize(
    Scale Factor=&Multiply_Vector2D(A=&RandomRangeVector2D(Minimum=<2 2>, Maximum=<3 3>), B=<1 1>)
    Scale Sprite Size Mode=::NewEnumerator2
   )
   &ScaleSpriteSize(
    Non-Uniform Curve Scale=<1 1.5>
    Non-Uniform Curve Sprite Scale=DataInterface() {propertyValues {
      XCurve {keys [
        {arriveTangent=0.8609, interpMode=RCIM_Cubic, leaveTangent=0.8609, tangentMode=RCTM_User, time=0, value=0.6}
        {arriveTangent=0.5103, interpMode=RCIM_Cubic, leaveTangent=0.5103, tangentMode=RCTM_User, time=0.3, value=0.8}
        {arriveTangent=0.0224, interpMode=RCIM_Cubic, leaveTangent=0.0224, tangentMode=RCTM_User, time=1, value=1}
       ]}
      YCurve {keys [
        {arriveTangent=0.8801, interpMode=RCIM_Cubic, leaveTangent=0.8801, tangentMode=RCTM_User, time=0, value=0.3}
        {arriveTangent=0.6586, interpMode=RCIM_Cubic, leaveTangent=0.6586, tangentMode=RCTM_User, time=0.3, value=0.6}
        {arriveTangent=0.0224, interpMode=RCIM_Cubic, leaveTangent=0.0224, tangentMode=RCTM_User, time=1, value=1}
       ]}
     }}
    Scale Sprite Size Mode=::NewEnumerator3
   ) {module=ScaleSpriteSize001}
   &ScaleSpriteSize(
    Scale Factor=&MakeVector2D(
     X=&FloatFromCurve(FloatCurve=DataInterface() {propertyValues {Curve {keys [
          {interpMode=RCIM_Cubic, tangentMode=RCTM_Break, time=0.2, value=0.8}
          {arriveTangent=3.8411, interpMode=RCIM_Cubic, tangentMode=RCTM_Break, time=1, value=3}
         ]}}})
     Y=&FloatFromCurve(FloatCurve=DataInterface() {propertyValues {Curve {keys [
          {interpMode=RCIM_Cubic, tangentMode=RCTM_Break, time=0.2, value=1}
          {arriveTangent=0.7569, interpMode=RCIM_Cubic, tangentMode=RCTM_Break, time=1, value=1.5}
         ]}}})
    )
    Scale Sprite Size Mode=::NewEnumerator2
   ) {module=ScaleSpriteSize003}
   &Drag(Drag=0, Use Linear Drag=true, Use Rotational Drag=true)
   &SolveForcesAndVelocity(Clamp Velocity=false, Limit Acceleration=false)
   &ScaleSpriteSizeBySpeed(Max Scale Factor=<1 4>)
   &ScaleColor(
    Scale Alpha=&FloatFromCurve(
     FloatCurve=DataInterface() {propertyValues {Curve {keys [
         {arriveTangent=9.0725, interpMode=RCIM_Cubic, leaveTangent=9.0725, tangentMode=RCTM_User, time=0, value=1}
         {arriveTangent=-2.5249, interpMode=RCIM_Cubic, leaveTangent=-2.5249, tangentMode=RCTM_User, time=0.1408, value=0.9847}
         {arriveTangent=-1.8607, interpMode=RCIM_Cubic, leaveTangent=-1.8607, tangentMode=RCTM_User, time=0.3012, value=0.9413}
         {arriveTangent=-1.0655, interpMode=RCIM_Cubic, leaveTangent=-1.0655, tangentMode=RCTM_User, time=0.5976, value=0.5305}
         {arriveTangent=-0.0405, interpMode=RCIM_Cubic, leaveTangent=-0.0405, tangentMode=RCTM_User, time=1, value=0}
        ]}}}
     Scale Curve=&Multiply_Float(
      A=2
      B=&self:OneWayTimelineScaler(OneWayTimeline:NiagaraFloat=$System.OneWayTimeline, Scale:NiagaraFloat=1)
     )
    )
    Scale RGB=&Multiply_VectorByFloat(
     Vector=&MakeVectorFromLinearColorRGB(Color=&VariantColorSelector(
       Variant Amount=0
       Original Color=&ColorFromCurve(ColorCurve=DataInterface() {propertyValues {
          AlphaCurve {keys [<at 0 1 RCIM_Cubic>,<at 1 1 RCIM_Cubic>]}
          BlueCurve {keys [
            <at 0.034 0>
            <at 0.1599 0>
            <at 0.3037 0 RCIM_Cubic>
            {arriveTangent=0.1755, interpMode=RCIM_Cubic, leaveTangent=0.1755, time=0.4429, value=0}
            <at 0.5885 0.05>
           ]}
          GreenCurve {keys [
            <at 0.034 0.5833>
            <at 0.1599 0.4793>
            {arriveTangent=-1.5634, interpMode=RCIM_Cubic, leaveTangent=-1.5634, time=0.3037, value=0.2667}
            {arriveTangent=-0.7285, interpMode=RCIM_Cubic, leaveTangent=-0.7285, time=0.4429, value=0.037}
            <at 0.5885 0.0592>
           ]}
          RedCurve {keys [
            <at 0.034 1>
            <at 0.1599 1>
            {arriveTangent=-1.9652, interpMode=RCIM_Cubic, leaveTangent=-1.9652, time=0.3037, value=1}
            {arriveTangent=-3.1598, interpMode=RCIM_Cubic, leaveTangent=-3.1598, time=0.4429, value=0.444}
            <at 0.5885 0.1>
           ]}
         }})
       Variant 1 Color=&ColorFromCurve(ColorCurve=DataInterface() {propertyValues {
          AlphaCurve {keys [<at 0 1 RCIM_Cubic>,<at 1 1 RCIM_Cubic>]}
          BlueCurve {keys [<at 0.034 0.2>, {arriveTangent=-0.1803, interpMode=RCIM_Cubic, leaveTangent=-0.1803, time=0.4429, value=0.1}, <at 0.5885 0.1>]}
          GreenCurve {keys [<at 0.034 0.0131>, {arriveTangent=-0.0237, interpMode=RCIM_Cubic, leaveTangent=-0.0237, time=0.4429, value=0.0128}, <at 0.5885 0>]}
          RedCurve {keys [<at 0.034 0>,<at 0.4429 0 RCIM_Cubic>,<at 0.5885 0>]}
         }})
       Variant 2 Color=?
      ))
     Float=&Multiply_Float(
      A=&Lerp_Float(
       A=0.5
       B=1
       Alpha=&FloatFromCurve(
        FloatCurve=DataInterface() {propertyValues {Curve {keys [
            {arriveTangent=0.0028, interpMode=RCIM_Cubic, leaveTangent=0.0028, tangentMode=RCTM_User, time=0, value=1}
            {arriveTangent=-2.5249, interpMode=RCIM_Cubic, leaveTangent=-2.5249, tangentMode=RCTM_User, time=0.1408, value=0.9847}
            {arriveTangent=-0.0405, interpMode=RCIM_Cubic, leaveTangent=-0.0405, tangentMode=RCTM_User, time=0.8, value=0}
           ]}}}
        Scale Curve=&Lerp_Float(A=1, B=3, Alpha=$System.EffectIntensity)
       )
      )
      B=&FNTimeOfDayScalar(Night Scale=0.5, Morning Scale=0.7, Evening Scale=0.7)
     )
    )
    ScaleA=true
    ScaleRGB=true
   ) {module=ScaleColor001}
   &DynamicMaterialParameters(
    Index 0 Param 1=&FloatFromCurve(FloatCurve=DataInterface() {propertyValues {Curve {keys [
         {arriveTangent=-0.1846, time=0, value=0}
         {arriveTangent=1.3385, interpMode=RCIM_Cubic, leaveTangent=1.1594, tangentMode=RCTM_Break, time=0.5, value=0}
         {arriveTangent=2, interpMode=RCIM_Cubic, leaveTangent=2, time=0.9, value=0.45}
         {arriveTangent=3.0091, interpMode=RCIM_Cubic, leaveTangent=3.0091, tangentMode=RCTM_User, time=1, value=1}
        ]}}})
    Index 0 Param 2=1
    Index 0 Param 3=&Lerp_Float(A=0, B=1, Alpha=&Multiply_Float(A=$System.Timeline.Time, B=1))
    Index 0 Param 4=&Lerp_Float(A=0, B=1, Alpha=&Multiply_Float(A=$System.Timeline.Time, B=1))
    Module.Index0Param0Write=true
    Module.Index0Param1Write=true
    Module.Index0Param2Write=true
    Module.Index0Param3Write=true
   )
  ]
  rendererClasses [&NiagaraSpriteRendererProperties]
  renderers [{rendererClass=&NiagaraSpriteRendererProperties}]
 }
 GlowsAndLights {
  emitterUpdateScript [
   &EmitterState(
    Inactive Response=::NewEnumerator1
    Life Cycle Mode=::NewEnumerator1
    Loop Behavior=::NewEnumerator1
    Loop Duration Mode=::NewEnumerator1
    UseLoopDelay=false
   )
   &SpawnBurst_Instantaneous(Spawn Count=4, Use Spawn Probability=false, UseLoopCountLimit=true)
  ]
  particleSpawnScript [
   &FN_SetSpawnGroup(Number of Spawn Groups:NiagaraInt32=$System.PipeAmount)
   &InitializeParticle(UsePositionOffset=true, Write Material Random=true) {module=InitializeParticle001}
  ]
  particleUpdateScript [
   &ParticleState(Kill Particles When Lifetime Has Elapsed=false)
   &SkeletalMeshLocation(
    Bone / Socket Index=&SelectIntFromArray(Direct Array Index=$Particles.SpawnGroup, Int Selection Array=DataInterface() {propertyValues {IntData [2,3,4]}})
    MaskSpawnGroup=false
    Mesh Sampling Type=::NewEnumerator5
    Skeletal Mesh=$System.SkeletalMesh
    Socket Sampling Mode=::NewEnumerator5
   )
   &ScaleColor(
    Scale Alpha=&Multiply_Float(A=2.5, B=$System.Timeline.Time)
    Scale RGB=&VectorFromFloat(Value=&Multiply_Float(A=$System.Flicker, B=3))
    ScaleA=true
    ScaleRGB=true
   )
   &FN_Light_Attributes(
    Color=&MultiplyLinearColor(
     Linear Color=&VariantColorSelector(Variant Amount=0, Original Color=<rgba 5 1.205 0.05 1>, Variant 1 Color=<rgba 0 0.626 5 1>, Variant 2 Color=?)
     Scale Factor=&MakeFloat4FromVectorAndFloat(
      "Vector (RGB)"=&VectorFromFloat(Value=16)
      "Float (Alpha)"=&Multiply_Float(
       A=2.5
       B=&self:OneWayTimelineScaler(OneWayTimeline:NiagaraFloat=$System.OneWayTimeline, Scale:NiagaraFloat=5)
      )
     )
    )
    Create Color Attribute=true
    Create Exponent Attribute=true
    Create Position Attribute=true
    Create Radius Attribute=true
    Position=&AddVectorToPosition(
     Position=$Particles.SkeletalMeshLocation.SampledPosition
     Vector=&TransformVector(Vector=<0 0 3>, Destination Space=::Simulation)
    )
    Radius=10
    Scale Color=$System.Flicker
   )
   /SetVariables(
    Particles.JetGlowCameraOffset=2
    Particles.JetGlowColor:LinearColor=&MultiplyLinearColor(
     Linear Color=&VariantColorSelector(Variant Amount=0, Original Color=<rgba 0.5 0.0575 0.05 0.3>, Variant 1 Color=<rgba 0 0.242 0.5 0.3>, Variant 2 Color=?)
     Scale Factor=&MakeFloat4FromVectorAndFloat(
      "Vector (RGB)"=<1 1 1>
      "Float (Alpha)"=&Multiply_Float(
       A=$System.Flicker
       B=&self:OneWayTimelineScaler(OneWayTimeline:NiagaraFloat=$System.OneWayTimeline, Scale:NiagaraFloat=1)
      )
     )
    )
    Particles.JetGlowPosition:NiagaraPosition=$Particles.Position
    Particles.JetGlowSpriteSize:Vector2f=&Vector2DFromFloat(Value=5)
   ) {isSetParameters=true}
   /SetVariables(
    Particles.FlameGlowCameraOffset=2
    Particles.FlameGlowColor:LinearColor=&MultiplyLinearColor(
     Linear Color=&VariantColorSelector(Variant Amount=0, Original Color=<rgba 0.6 0.069 0.06 0.6>, Variant 1 Color=<rgba 0 0.0981 0.6 0.6>, Variant 2 Color=?)
     Scale Factor=&MakeFloat4FromVectorAndFloat(
      "Vector (RGB)"=&VectorFromFloat(Value=&Multiply_Float(
        A=$System.Flicker
        B=&MakeCustomFloatFromBool(Bool=$User.bIsFrontEndPreview, True Float=0.4, False Float=1)
       ))
      "Float (Alpha)"=&Multiply_Float(
       A=&Multiply_Float(
        A=0.6
        B=&self:OneWayTimelineScaler(OneWayTimeline:NiagaraFloat=$System.OneWayTimeline, Scale:NiagaraFloat=1)
       )
       B=&FNTimeOfDayScalar(Night Scale=0.5, Morning Scale=0.7, Evening Scale=0.7)
      )
     )
    )
    Particles.FlameGlowPosition:NiagaraPosition=$Particles.Position
    Particles.FlameGlowSpriteSize:Vector2f=<2.5 6>
    Particles.FlameGlowVelocity:Vector3f=&RotateVectorByQuaternion(Vector To Rotate=<0 0 1>, Quaternion=$Particles.SkeletalMeshLocation.SampledRotation)
   ) {isSetParameters=true}
   /SetVariables(
    Particles.GlowCardColor:LinearColor=&MultiplyLinearColor(
     Linear Color=&VariantColorSelector(Variant Amount=0, Original Color=<rgba 100 10 0 1>, Variant 1 Color=<rgba 0 83.6959 100 1>, Variant 2 Color=?)
     Scale Factor=&MakeFloat4FromVectorAndFloat(
      "Vector (RGB)"=&VectorFromFloat(Value=$System.Flicker)
      "Float (Alpha)"=&Multiply_Float(
       A=1.5
       B=&self:OneWayTimelineScaler(OneWayTimeline:NiagaraFloat=$System.OneWayTimeline, Scale:NiagaraFloat=1)
      )
     )
    )
    Particles.GlowCardPosition:NiagaraPosition=&AddVectorToPosition(
     Position=$Particles.Position
     Vector=&SelectVectorFromArray(Direct Array Index=$Particles.SpawnGroup, Vector Selection Array=DataInterface() {propertyValues {FloatData [<0 0 0>, <0 0 0>]}})
    )
    Particles.GlowCardSpriteFacing:Vector3f=<0 1 0>
    Particles.GlowCardSpriteSize:Vector2f=<15 15>
   ) {isSetParameters=true}
   /SetVariables(
    Particles.InnerGlowCameraOffset:NiagaraFloat=&FN_FrontEndFloatModifier(Gameplay=0, Front End=2)
    Particles.InnerGlowColor:LinearColor=&MultiplyLinearColor(
     Linear Color=&VariantColorSelector(Variant Amount=0, Original Color=<rgba 9 2.4 0 0.5>, Variant 1 Color=<rgba 0.0423 0 8 0.5>, Variant 2 Color=?)
     Scale Factor=&MakeFloat4FromVectorAndFloat(
      "Vector (RGB)"=<1 1 1>
      "Float (Alpha)"=&Multiply_Float(
       A=&self:OneWayTimelineScaler(OneWayTimeline:NiagaraFloat=$System.OneWayTimeline, Scale:NiagaraFloat=1)
       B=1
      )
     )
    )
    Particles.InnerGlowDMP1:Vector4f=<5 0.5 1 1>
    Particles.InnerGlowPosition:NiagaraPosition=$Particles.Position
    Particles.InnerGlowSpriteSize:Vector2f=&Vector2DFromFloat(Value=6)
   ) {isSetParameters=true}
  ]
  rendererClasses [&NiagaraLightRendererProperties, &NiagaraSpriteRendererProperties]
  renderers [
   {rendererClass=&NiagaraLightRendererProperties, rendererIndex=0}
   {rendererClass=&NiagaraSpriteRendererProperties, rendererIndex=1}
   {rendererClass=&NiagaraSpriteRendererProperties, rendererIndex=2}
   {rendererClass=&NiagaraSpriteRendererProperties, rendererIndex=3}
  ]
 }
 Smoke001 {
  emitterUpdateScript [
   &EmitterState()
   &SpawnRate(
    SpawnRate=&Lerp_Float(
     A=&FN_PickFloatDependOnPlatform(TargetPlatform=DataInterface() {propertyValues {Platforms {qualityLevelMask=-29}}}, TargetPlatformActive=50, TargetPlatformNoActive=130)
     B=180
     Alpha=$System.Timeline.Time
    )
    Use Spawn Probability=true
   )
  ]
  particleSpawnScript [
   &PartitionParticles(Number Of Partitions=$System.PipeAmount, Write Partition to Ribbon ID=::NewEnumerator0)
   &InitializeParticle(
    Color=&MultiplyLinearColor(
     Linear Color=<rgba 1 1 1 1>
     Scale Factor=&MakeVector4_FromFloat(Float=&FN_FrontEndFloatModifier(Gameplay=1, Front End=1))
    )
    Lifetime=0.67
    Mesh Scale=&Multiply_VectorByFloat(Vector=&RandomRangeVector(Minimum=<3.5 2 2.5>, Maximum=<4.5 3.5 3.5>), Float=0.02)
    Mesh Scale Mode=::NewEnumerator4
    UsePositionOffset=false
    Write Material Random=true
   )
   &SkeletalMeshLocation(
    Bone / Socket Index=&SelectIntFromArray(Direct Array Index=$Particles.PartitionParticles.Partition, Int Selection Array=DataInterface() {propertyValues {IntData [2,3,4]}})
    MaskSpawnGroup=false
    Mesh Sampling Type=::NewEnumerator5
    Orientation Sampling=::NewEnumerator1
    Skeletal Mesh=$System.SkeletalMesh
    Socket Sampling Mode=::NewEnumerator5
   ) {module=SkeletalMeshLocation001}
   &InitialMeshOrientation(
    Mesh Orientation Mode=::NewEnumerator2
    Rotation=&RandomRangeVector(Minimum=<0 0 0>, Maximum=<1 0 0>)
    Use Rotation Vector=true
   )
   &CameraOffset(Camera Offset Amount=-10)
  ]
  particleUpdateScript [
   &ParticleState()
   &SkeletalMeshLocation(
    Bone / Socket Index=&SelectIntFromArray(Direct Array Index=$Particles.PartitionParticles.Partition, Int Selection Array=DataInterface() {propertyValues {IntData [2,3,4]}})
    MaskSpawnGroup=false
    Mesh Sampling Type=::NewEnumerator5
    Orientation Sampling=::NewEnumerator1
    Position Sampling=::NewEnumerator2
    Skeletal Mesh=$System.SkeletalMesh
    Socket Sampling Mode=::NewEnumerator5
   ) {module=SkeletalMeshLocation002}
   &InheritSourceMovement(
    Applied Force Scale=&Lerp_Vector(
     A=&VectorFromFloat(Value=&FloatFromCurve(FloatCurve=DataInterface() {propertyValues {Curve {keys [<at 0.2 1>,<at 0.4 0>]}}}))
     B=&VectorFromFloat(Value=0.5)
     Alpha=$System.Timeline001.NormalizedTime
    )
    Applied Position Scale=&Multiply_VectorByFloat(
     Vector=<1 1 1>
     Float=&FloatFromCurve(FloatCurve=DataInterface() {propertyValues {Curve {keys [
          <at 0 1 RCIM_Cubic>
          {arriveTangent=-0.5, interpMode=RCIM_Cubic, leaveTangent=-0.5, time=0.25, value=1}
          {arriveTangent=-0.0269, interpMode=RCIM_Cubic, leaveTangent=-0.0269, tangentMode=RCTM_User, time=1, value=0.5}
         ]}}})
    )
    ApplyForce=true
    ApplyForceFalloff=false
    ApplyPosition=true
    ApplyPositionFalloff=false
    Module.LimitForceSpeed=false
    Source Velocity:Vector3f=$System.SystemVelocity
   )
   &Drag(Use Linear Drag=true, Use Rotational Drag=true)
   &GravityForce(Gravity=<0 0 150>)
   &SolveForcesAndVelocity(Clamp Velocity=false, Limit Acceleration=false)
   &ScaleMeshSize(Scale Factor=&Multiply_VectorByFloat(
     Vector=&Multiply_VectorByFloat(
      Vector=&Multiply_Vector(
       A=<1 1 1>
       B=&VectorFromCurve(VectorCurve=DataInterface() {propertyValues {XCurve {keys [<at 0 5>,<at 0.2 1>,<at 1 1>]}, YCurve {keys [<at 0 1>,<at 1 1>]}, ZCurve {keys [<at 0 1>,<at 1 1>]}}})
      )
      Float=&RandomRangeFloat(Minimum=1, Maximum=1.5)
     )
     Float=&FloatFromCurve(
      FloatCurve=DataInterface() {propertyValues {Curve {keys [
          {interpMode=RCIM_Cubic, tangentMode=RCTM_User, time=0, value=0}
          {arriveTangent=1.3433, tangentMode=RCTM_User, time=0.2, value=0.35}
          {arriveTangent=1.4526, interpMode=RCIM_Cubic, leaveTangent=1.4526, tangentMode=RCTM_User, time=1, value=1.5}
         ]}}}
      Scale Curve=&Multiply_Float(A=1, B=1.5)
     )
    ))
   &DynamicMaterialParameters(
    Index 0 Param 1=&FloatFromCurve(
     FloatCurve=DataInterface() {propertyValues {Curve {keys [
         {interpMode=RCIM_Cubic, tangentMode=RCTM_User, time=0, value=0}
         <at 0.2 0.55>
         {arriveTangent=-1.0223, interpMode=RCIM_Cubic, leaveTangent=-1.0223, tangentMode=RCTM_User, time=1, value=1}
        ]}}}
     Scale Curve=&RandomRangeFloat(Minimum=7, Maximum=15)
    )
    Index 0 Param 4=&Lerp_Float(
     A=0.6
     B=0
     Alpha=&FloatFromCurve(FloatCurve=DataInterface() {propertyValues {Curve {keys [
          {arriveTangent=0.7438, interpMode=RCIM_Cubic, leaveTangent=0.7438, tangentMode=RCTM_User, time=0, value=0}
          {arriveTangent=0.0023, interpMode=RCIM_Cubic, leaveTangent=0.0023, tangentMode=RCTM_User, time=1, value=1}
         ]}}})
    )
    Module.Index0Param0Write=true
    Module.Index0Param1Write=true
    Module.Index0Param2Write=true
    Module.Index0Param3Write=true
   ) {module=DynamicMaterialParameters001}
   /SetVariables(
    Particles.DY0:Vector4f=<4 0 0 0>
    Particles.MaterialRandom:NiagaraFloat=&SelectFloatFromArray(Direct Array Index=$Particles.PartitionParticles.Partition, Float Selection Array=DataInterface() {propertyValues {FloatData [0,0.2,0.4]}})
    Particles.Position=&Lerp_Vector(
     A=&AddVectorToPosition(
      Position=$Particles.SkeletalMeshLocation002.SampledPosition
      Vector=&RotateVectorByQuaternion(
       Vector To Rotate=&VectorFromCurve(
        VectorCurve=DataInterface() {propertyValues {XCurve {keys [<at 0 0>,<at 1 0>]}, YCurve {keys [<at 0 0>,<at 1 0>]}, ZCurve {keys [<at 0 0>,<at 1 1>]}}}
        Scale Curve=<1 1 10>
       )
       Quaternion=$Particles.SkeletalMeshLocation002.SampledRotation
      )
     )
     B=Linked() {linkedVariable {name=Particles.Position, type=NiagaraPosition}}
     Alpha=&FloatFromCurve(FloatCurve=DataInterface() {propertyValues {Curve {keys [{interpMode=RCIM_Cubic, tangentMode=RCTM_User, time=0, value=0.2}, {interpMode=RCIM_Cubic, tangentMode=RCTM_User, time=0.7, value=1}]}}})
    )
    Particles.RibbonColor:LinearColor=&MultiplyLinearColor(
     Linear Color=&MultiplyLinearColor(
      Linear Color=&VariantColorSelector(
       Variant Amount=0
       Original Color=&ColorFromCurve(
        ColorCurve=DataInterface() {propertyValues {
          AlphaCurve {keys [<at 0.1579 2>,<at 0.656 1>,<at 0.8476 0>]}
          BlueCurve {keys [<at 0 2.5 RCIM_Cubic>, <at 0.1569 0>, <at 0.2559 0>, {interpMode=RCIM_Cubic, leaveTangent=2.8703, time=0.5591, value=0}]}
          GreenCurve {keys [<at 0 8.75 RCIM_Cubic>, <at 0.1569 44.196>, <at 0.2559 8.6667>, {interpMode=RCIM_Cubic, leaveTangent=-14.0733, time=0.5591, value=0.1667}]}
          RedCurve {keys [<at 0 40 RCIM_Cubic>, <at 0.1569 80>, <at 0.2559 40>, {interpMode=RCIM_Cubic, leaveTangent=-52.566, time=0.5591, value=0.5}]}
         }}
        CurveIndex=&FloatFromCurve(
         FloatCurve=DataInterface() {propertyValues {Curve {keys [
             {arriveTangent=0.0922, interpMode=RCIM_Cubic, leaveTangent=0.0922, tangentMode=RCTM_User, time=1, value=0}
             {arriveTangent=0.0029, interpMode=RCIM_Cubic, leaveTangent=0.0029, tangentMode=RCTM_User, time=50, value=1}
            ]}}}
         CurveIndex=&DistanceBetweenPositions(A=$Particles.SkeletalMeshLocation002.SampledPosition, B=$Particles.Position)
        )
       )
       Variant 1 Color=&ColorFromCurve(
        ColorCurve=DataInterface() {propertyValues {
          AlphaCurve {keys [<at 0.1579 3>,<at 0.656 1>,<at 0.8476 0>]}
          BlueCurve {keys [<at 0.0855 80 RCIM_Cubic>, {interpMode=RCIM_Cubic, leaveTangent=-101.2855, time=0.1628, value=30}]}
          GreenCurve {keys [<at 0.0855 36.4259 RCIM_Cubic>, {interpMode=RCIM_Cubic, leaveTangent=-47.7506, time=0.1628, value=4.7235}]}
          RedCurve {keys [<at 0.0855 0 RCIM_Cubic>,<at 0.1628 0 RCIM_Cubic>]}
         }}
        CurveIndex=&FloatFromCurve(
         FloatCurve=DataInterface() {propertyValues {Curve {keys [
             {arriveTangent=0.0922, interpMode=RCIM_Cubic, leaveTangent=0.0922, tangentMode=RCTM_User, time=1, value=0}
             {arriveTangent=0.0029, interpMode=RCIM_Cubic, leaveTangent=0.0029, tangentMode=RCTM_User, time=50, value=1}
            ]}}}
         CurveIndex=&DistanceBetweenPositions(A=$Particles.SkeletalMeshLocation002.SampledPosition, B=$Particles.Position)
        )
       )
       Variant 2 Color=?
      )
      Scale Factor=&MakeFloat4FromVectorAndFloat(
       "Vector (RGB)"=&VectorFromFloat(Value=&self:OneWayTimelineScaler(OneWayTimeline:NiagaraFloat=$System.OneWayTimeline, Scale:NiagaraFloat=1))
       "Float (Alpha)"=1
      )
     )
     Scale Factor=&MakeVector4(X=1, Y=1, Z=1, W=$System.Timeline.NormalizedTime)
    )
    Particles.RibbonSmokeColor:LinearColor=&MultiplyLinearColor(
     Linear Color=&VariantColorSelector(
      Variant Amount=0
      Original Color=&ColorFromCurve(ColorCurve=DataInterface() {propertyValues {
         AlphaCurve {keys [<at 0.1579 4>,<at 0.656 1>,<at 0.8476 0>]}
         BlueCurve {keys [<at 0 2.5 RCIM_Cubic>,<at 0.1584 0 RCIM_Cubic>]}
         GreenCurve {keys [<at 0 8.75 RCIM_Cubic>,<at 0.1584 0.1667 RCIM_Cubic>]}
         RedCurve {keys [<at 0 40 RCIM_Cubic>,<at 0.1584 0.5 RCIM_Cubic>]}
        }})
      Variant 1 Color=&ColorFromCurve(ColorCurve=DataInterface() {propertyValues {
         AlphaCurve {keys [<at 0.1579 4>,<at 0.656 1>,<at 0.8476 0>]}
         BlueCurve {keys [<at 0 40 RCIM_Cubic>,<at 0.1584 0.5 RCIM_Cubic>]}
         GreenCurve {keys [<at 0 28.3605 RCIM_Cubic>,<at 0.1584 0.2391 RCIM_Cubic>]}
         RedCurve {keys [<at 0 0 RCIM_Cubic>,<at 0.1584 0 RCIM_Cubic>]}
        }})
      Variant 2 Color=?
     )
     Scale Factor=&MakeVector4(
      X=1
      Y=1
      Z=1
      W=&Multiply_Float(
       A=$System.Timeline.NormalizedTime
       B=&FloatFromCurve(
        FloatCurve=DataInterface() {propertyValues {Curve {keys [{arriveTangent=0.0922, time=25, value=1}, <at 30 0>]}}}
        CurveIndex=&DistanceBetweenPositions(A=$Particles.SkeletalMeshLocation002.SampledPosition, B=$Particles.Position)
       )
      )
     )
    )
    Particles.RibbonSmokeW:NiagaraFloat=&FloatFromCurve(FloatCurve=DataInterface() {propertyValues {Curve {keys [<at 0 3>,<at 1 6>]}}}, Scale Curve=2)
    Particles.RibbonWidth=&FloatFromCurve(FloatCurve=DataInterface() {propertyValues {Curve {keys [<at 0 1>,<at 1 6>]}}}, Scale Curve=2)
   ) {isSetParameters=true, module=SetVariables_#5}
  ]
  rendererClasses [&NiagaraRibbonRendererProperties]
  renderers [
   {rendererClass=&NiagaraRibbonRendererProperties}
   {rendererClass=&NiagaraRibbonRendererProperties, rendererIndex=1}
  ]
 }
 Sparks {
  emitterUpdateScript [
   &EmitterState()
   &SpawnRate(
    SpawnRate=&self:OneWayTimelineScaler(OneWayTimeline:NiagaraFloat=$System.OneWayBurst, Scale:NiagaraFloat=10)
    Use Spawn Probability=false
   )
  ]
  particleSpawnScript [
   &PartitionParticles(Number Of Partitions=$System.PipeAmount, Write Partition to Ribbon ID=::NewEnumerator0)
   /SetVariables(Particles.PerParticleRandom:NiagaraFloat=&FloatFromCurve(
     FloatCurve=DataInterface() {propertyValues {Curve {keys [<at 0 0>,<at 0.7 0.4>,<at 1 1>]}}}
     CurveIndex=&RandomRangeFloat(Minimum=0, Maximum=1)
    )) {isSetParameters=true}
   &InitializeParticle(
    Color=&MultiplyLinearColor(
     Linear Color=<rgba 1 1 1 1>
     Scale Factor=&MakeVector4_FromFloat(Float=&FN_FrontEndFloatModifier(Gameplay=1, Front End=1))
    )
    Lifetime=&Multiply_Float(
     A=0.5
     B=&UniformAOrBFloat(
      A=0.6
      B=0.35
      Distribution Weight=$Particles.PerParticleRandom
      Evaluation Type=::NewEnumerator0
      Recalculate Random Each Loop=true
      Override Seed=false
     )
    )
    Mesh Scale=&Multiply_VectorByFloat(Vector=&RandomRangeVector(Minimum=<3.5 2 2.5>, Maximum=<4.5 3.5 3.5>), Float=0.02)
    Mesh Scale Mode=::NewEnumerator4
    Sprite Size=&Multiply_Vector2D_ByFloat(Float=&Lerp_Float(A=0.5, B=1.5, Alpha=$Particles.PerParticleRandom), Vector2D=<1 4>)
    Sprite Size Mode=::NewEnumerator4
    UsePositionOffset=false
    Write Material Random=true
   )
   &SkeletalMeshLocation(
    Bone / Socket Index=&SelectIntFromArray(Direct Array Index=$Particles.PartitionParticles.Partition, Int Selection Array=DataInterface() {propertyValues {IntData [2,3,4]}})
    MaskSpawnGroup=false
    Mesh Sampling Type=::NewEnumerator5
    Orientation Sampling=::NewEnumerator1
    Skeletal Mesh=$System.SkeletalMesh
    Socket Sampling Mode=::NewEnumerator5
   ) {module=SkeletalMeshLocation001}
   &CameraOffset(Camera Offset Amount=-10)
   &AddVelocity(
    Cone Angle=20
    Cone Axis=<0 0.2 1>
    Module.UseConeFalloff=false
    Module.UseVelDistribution=false
    Override Seed=false
    Velocity Mode=::NewEnumerator2
    Velocity Speed=&Multiply_Float(
     A=1.5
     B=&UniformAOrBFloat(
      A=300
      B=700
      Distribution Weight=$Particles.PerParticleRandom
      Evaluation Type=::NewEnumerator0
      Recalculate Random Each Loop=true
      Override Seed=false
     )
    )
   )
  ]
  particleUpdateScript [
   &ParticleState()
   &SkeletalMeshLocation(
    Bone / Socket Index=&SelectIntFromArray(Direct Array Index=$Particles.PartitionParticles.Partition, Int Selection Array=DataInterface() {propertyValues {IntData [2,3,4]}})
    MaskSpawnGroup=false
    Mesh Sampling Type=::NewEnumerator5
    Orientation Sampling=::NewEnumerator1
    Position Sampling=::NewEnumerator2
    Sampled Position Offset=&RandomRangeVector(Minimum=<-1 -1 0>, Maximum=<1 1 0>)
    Skeletal Mesh=$System.SkeletalMesh
    Socket Sampling Mode=::NewEnumerator5
   ) {module=SkeletalMeshLocation002}
   &InheritSourceMovement(
    Applied Force Scale=&VectorFromFloat(Value=&FloatFromCurve(FloatCurve=DataInterface() {propertyValues {Curve {keys [<at 0.2 1>,<at 0.4 0>]}}}))
    Applied Position Scale=&Multiply_VectorByFloat(
     Vector=<1 1 1>
     Float=&FloatFromCurve(FloatCurve=DataInterface() {propertyValues {Curve {keys [
          <at 0 1 RCIM_Cubic>
          {arriveTangent=-0.5, interpMode=RCIM_Cubic, leaveTangent=-0.5, time=0.25, value=1}
          {arriveTangent=-0.0269, interpMode=RCIM_Cubic, leaveTangent=-0.0269, tangentMode=RCTM_User, time=1, value=0.5}
         ]}}})
    )
    ApplyForce=true
    ApplyForceFalloff=false
    ApplyPosition=true
    ApplyPositionFalloff=false
    Module.LimitForceSpeed=false
    Source Velocity:Vector3f=$System.SystemVelocity
   )
   &Drag(
    Drag=&FloatFromCurve(
     FloatCurve=DataInterface() {propertyValues {Curve {keys [<at 0 0>,<at 0.5 1>]}}}
     Scale Curve=&Lerp_Float(A=40, B=35, Alpha=$Particles.PerParticleRandom)
    )
    Use Linear Drag=true
    Use Rotational Drag=true
   )
   &GravityForce(Gravity=<0 0 150>)
   &SolveForcesAndVelocity(Clamp Velocity=false, Limit Acceleration=false)
   &ScaleColor(
    Scale Alpha=&FloatFromCurve(
     FloatCurve=DataInterface() {propertyValues {Curve {keys [
         {arriveTangent=2.8927, interpMode=RCIM_Cubic, leaveTangent=2.8927, tangentMode=RCTM_User, time=0, value=0}
         <at 0.3 1>
         {interpMode=RCIM_Cubic, tangentMode=RCTM_Break, time=0.5, value=1}
         <at 0.8 0>
        ]}}}
     Scale Curve=&Multiply_Float(A=1.55, B=$System.Timeline.Time)
    )
    Scale RGB=&MakeVectorFromLinearColorRGB(Color=&VariantColorSelector(
      Variant Amount=0
      Original Color=&ColorFromCurve(
       ColorCurve=DataInterface() {propertyValues {
         AlphaCurve {keys []}
         BlueCurve {keys [<at 0.1515 2.5 RCIM_Cubic>,<at 0.4671 0 RCIM_Cubic>]}
         GreenCurve {keys [<at 0.1515 8.75 RCIM_Cubic>,<at 0.4671 0.1667 RCIM_Cubic>]}
         RedCurve {keys [<at 0.1515 40 RCIM_Cubic>,<at 0.4671 0.5 RCIM_Cubic>]}
        }}
       CurveIndex=&Add_Float(A=$Particles.NormalizedAge, B=&Lerp_Float(A=0, B=0.4, Alpha=$Particles.PerParticleRandom))
      )
      Variant 1 Color=&ColorFromCurve(
       ColorCurve=DataInterface() {propertyValues {
         AlphaCurve {keys []}
         BlueCurve {keys [<at 0.1515 40 RCIM_Cubic>,<at 0.4671 0.5 RCIM_Cubic>]}
         GreenCurve {keys [<at 0.1515 31.4638 RCIM_Cubic>,<at 0.4671 0 RCIM_Cubic>]}
         RedCurve {keys [<at 0.1515 0 RCIM_Cubic>,<at 0.4671 0.025 RCIM_Cubic>]}
        }}
       CurveIndex=&Add_Float(A=$Particles.NormalizedAge, B=&Lerp_Float(A=0, B=0.4, Alpha=$Particles.PerParticleRandom))
      )
      Variant 2 Color=?
     ))
    ScaleA=true
    ScaleRGB=true
   )
   /SetVariables(Particles.Position=&Lerp_Vector(
     A=&AddVectorToPosition(
      Position=$Particles.SkeletalMeshLocation002.SampledPosition
      Vector=&RotateVectorByQuaternion(
       Vector To Rotate=&VectorFromCurve(
        VectorCurve=DataInterface() {propertyValues {XCurve {keys [<at 0 0>,<at 1 0>]}, YCurve {keys [<at 0 0>,<at 1 0>]}, ZCurve {keys [<at 0 0>,<at 1 1>]}}}
        Scale Curve=<1 1 10>
       )
       Quaternion=$Particles.SkeletalMeshLocation002.SampledRotation
      )
     )
     B=Linked() {linkedVariable {name=Particles.Position, type=NiagaraPosition}}
     Alpha=&FloatFromCurve(FloatCurve=DataInterface() {propertyValues {Curve {keys [{interpMode=RCIM_Cubic, tangentMode=RCTM_User, time=0, value=0.2}, {interpMode=RCIM_Cubic, tangentMode=RCTM_User, time=0.7, value=1}]}}})
    )) {isSetParameters=true}
   &ScaleSpriteSize(
    Non-Uniform Curve Sprite Scale=DataInterface() {propertyValues {
      XCurve {keys [{tangentMode=RCTM_User, time=0, value=1}, {tangentMode=RCTM_User, time=1, value=0.5}]}
      YCurve {keys [
        {interpMode=RCIM_Cubic, tangentMode=RCTM_User, time=0, value=0}
        {interpMode=RCIM_Cubic, tangentMode=RCTM_User, time=0.1, value=1}
        {interpMode=RCIM_Cubic, tangentMode=RCTM_User, time=1, value=0}
       ]}
     }}
    Scale Sprite Size Mode=::NewEnumerator3
   )
   &SpriteFacingAndAlignment(
    Alignment Coordinate Space=::World
    Sprite Alignment=&NormalizeVector(
     Vector To Normalize=&TransformVector(
      Vector=&Lerp_Vector(
       A=<0 0 1>
       B=<0 -1 0>
       Alpha=&Lerp_Float(
        A=&FloatFromCurve(FloatCurve=DataInterface() {propertyValues {Curve {keys [<at 0.4 0>,<at 0.5 1>]}}})
        B=&FloatFromCurve(FloatCurve=DataInterface() {propertyValues {Curve {keys [<at 0.1 0>,<at 0.5 0>]}}})
        Alpha=$Particles.PerParticleRandom
       )
      )
      Destination Space=::World
     )
     Safe Normalize=true
     Safe Fallback Threshold=0
     Safe Fallback Vector=<0 0 0.001>
    )
    WriteAlignment=true
    WriteFacing=false
   )
   &ScaleSpriteSizeBySpeed(
    Max Scale Factor=<1.666 5>
    Min Scale Factor=<1.333 1.333>
    Sample Scale Factor By Curve=true
    Scale Factor Curve=DataInterface() {propertyValues {Curve {keys [<at 0 0>, {arriveTangent=1.4989, interpMode=RCIM_Cubic, tangentMode=RCTM_Break, time=1, value=1}]}}}
    Velocity Threshold=999
   ) {enabled=false}
  ]
  rendererClasses [&NiagaraSpriteRendererProperties]
  renderers [{rendererClass=&NiagaraSpriteRendererProperties}]
 }
}
summary {
 systemName=NS_Backpack_QuickBurst
 emitters [
  {emitterName=EyeGlowStack, rendererClasses [&NiagaraSpriteRendererProperties, &NiagaraLightRendererProperties]}
  {
   emitterName=GlowsAndLights
   rendererClasses [&NiagaraLightRendererProperties, &NiagaraSpriteRendererProperties]
  }
  {emitterName=Flames, rendererClasses [&NiagaraSpriteRendererProperties]}
  {emitterName=Smoke001, rendererClasses [&NiagaraRibbonRendererProperties]}
  {emitterName=Sparks, rendererClasses [&NiagaraSpriteRendererProperties]}
  {emitterName=End_Smoke, rendererClasses [&NiagaraMeshRendererProperties]}
 ]
 userVariables [
  User.bDebug {type=NiagaraBool}
  User.bIsFrontEnd {type=NiagaraBool}
  User.bIsFrontEndPreview {type=NiagaraBool}
  User.PlayerJumping {type=NiagaraBool}
  User.SprintReactive {type=NiagaraBool}
  User.Variant {type=NiagaraFloat, defaultValue=1, description=""}
 ]
}
