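# Animation Blueprint for the QuickBurst backpack cosmetic that poses its jaw and adds secondary motion to two pipes.
# The AnimGraph additively nudges the pack bone and pitches dyn_jaw through a bool-alpha Modify Bone.
# It then chains spring-driven, angle-limited AnimDynamics on the left and right pipe bones with custom gravity.
# Each update sets bIsJawOpen from tactical sprint, knee slide or slide flags, though no graph visibly reads it.
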AnimBlueprint /BRCosmetics/Accessories/FORT_Backpacks/Backpack_M_MED_QuickBurst/Meshes/M_MED_QuickBurst_Pack_AnimBP
target_skeleton=&/M_MED_QuickBurst_Pack_Skeleton
graphs [
 AnimGraph {nodes [
   n1 = "Transform (Modify) Bone - Bone: pack"(ComponentPose=Local To Component(LocalPose=Input Pose()), Translation="0,-1,0", Rotation="0,0,0") {
    translationMode=BMM_Additive
    translationSpace=BCS_BoneSpace
    alphaBoolBlend {blendInTime=0.2, blendOption=Linear, blendOutTime=0.2}
    boneToModify {boneName=pack}
    translation {y=-1}
   }
   n2 = "Transform (Modify) Bone - Bone: dyn_jaw"!(ComponentPose=@n1, bAlphaBoolEnabled:bool=True, Translation="0,0,0", Rotation="-35,0,0", Scale="1,1,1") -> (Pose) {
    alphaInputType=Bool
    rotationMode=BMM_Additive
    rotationSpace=BCS_BoneSpace
    alphaBoolBlend {blendInTime=0.2, blendOption=Linear, blendOutTime=0.2}
    boneToModify {boneName=dyn_jaw}
    rotation {pitch=-35}
   }
   n3 = Anim Dynamics - Bone: dyn_pipe_l!(ComponentPose=@n2, Alpha=1) -> (Pose) {
    angularDampingOverride=0.9
    angularSpringConstant=1
    bAngularSpring=true
    bLinearSpring=true
    bOverrideAngularDamping=true
    bOverrideLinearDamping=true
    bUseGravityOverride=true
    linearDampingOverride=0.99
    linearSpringConstant=1
    alphaBoolBlend {blendOption=Linear}
    boundBone {boneName=dyn_pipe_l}
    componentAppliedLinearAccClamp=<2000 2000 2000>
    componentLinearAccScale=<0.2 0.2 0.2>
    componentLinearVelScale=<0.15 0.15 0.15>
    gravityOverride {z=-350}
    physicsBodyDefinitions [{
      boundBone {boneName=dyn_pipe_l}
      boxExtents=<5 10 10>
      constraintSetup {angularLimitsMax=<0 20 0>, angularLimitsMin=<0 -20 0>, angularTarget=<0 0 0>}
      localJointOffset=<2 0 0>
     }]
    relativeSpaceBone {boneName=root}
   }
   n4 = Anim Dynamics - Bone: dyn_pipe_r!(ComponentPose=@n3, Alpha=1) -> (Pose) {
    angularDampingOverride=0.9
    angularSpringConstant=1
    bAngularSpring=true
    bLinearSpring=true
    bOverrideAngularDamping=true
    bOverrideLinearDamping=true
    bUseGravityOverride=true
    linearDampingOverride=0.99
    linearSpringConstant=1
    alphaBoolBlend {blendOption=Linear}
    boundBone {boneName=dyn_pipe_r}
    componentAppliedLinearAccClamp=<2000 2000 2000>
    componentLinearAccScale=<0.2 0.2 0.2>
    componentLinearVelScale=<0.15 0.15 0.15>
    gravityOverride {z=-350}
    physicsBodyDefinitions [{
      boundBone {boneName=dyn_pipe_r}
      boxExtents=<5 10 10>
      constraintSetup {angularLimitsMax=<0 20 0>, angularLimitsMin=<0 -20 0>, angularTarget=<0 0 0>}
      localJointOffset=<2 0 0>
     }]
    relativeSpaceBone {boneName=root}
   }
   Output Pose(Result=Component To Local(ComponentPose=@n4)) {layerGroup=None}
  ]}
 EventGraph {nodes [
   Event BlueprintUpdateAnimation()
   > Set bIsJawOpen(bIsJawOpen:bool=OR Boolean!(A=Get bIsTacticalSprinting(), B=Get bIsKneeSliding(), C:bool=Get bIsSliding()) -> (ReturnValue)) -> (Output_Get:bool)
   TryGetPawnOwner()
  ]}
]
variables [bIsJawOpen {type=boolean}]
